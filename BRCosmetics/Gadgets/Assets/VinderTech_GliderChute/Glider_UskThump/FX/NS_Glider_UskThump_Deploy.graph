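# Niagara deploy effect for the UskThump glider that bursts tumbling mesh pages from its left and right drawers.
# Each emitter fires three staggered spawn groups from its own sockets, with the third group smaller and tinted.
# Curve-driven forces, drag, curl noise and gravity move pages, with User.ForwardDot steering sideways push.
# Pages inherit source movement, take time-of-day color, and vary scale between gameplay and front end.
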NiagaraSystem /BRCosmetics/Gadgets/Assets/VinderTech_GliderChute/Glider_UskThump/FX/NS_Glider_UskThump_Deploy
dependencies {
 usedDataInterfaces [
  &NiagaraDataInterfaceArrayColor
  &NiagaraDataInterfaceMeshRendererInfo
  &NiagaraDataInterfaceArrayInt32
  &NiagaraDataInterfaceCurve
  &NiagaraDataInterfaceSocketReader
 ]
 usedDynamicInputs [&FNTimeOfDayColor]
 usedModules [
  &self:SystemSpawnScript.SystemScriptSource.SystemScriptGraph.NiagaraNodeAssignment_0.SetVariables_#1
  &self:SystemSpawnScript.SystemScriptSource.SystemScriptGraph.NiagaraNodeAssignment_3.SetVariables_#2
  &SystemState
  &self:SystemSpawnScript.SystemScriptSource.SystemScriptGraph.NiagaraNodeAssignment_4.SetVariables_#3
 ]
}
emitters {
 Left_Drawers_Pages {
  emitterUpdateScript [
   &EmitterState(
    Inactive Response=::NewEnumerator1
    Life Cycle Mode=::NewEnumerator1
    Loop Behavior=::NewEnumerator1
    Loop Delay=$System.LoopDelay
    Loop Duration=2
    UseLoopDelay=true
   )
   &SpawnBurst_Instantaneous(
    Spawn Count=&RandomRangeInt(Minimum=3, Maximum=4)
    Spawn Time=0.015
    Use Spawn Probability=false
    UseLoopCountLimit=false
   )
   &SpawnBurst_Instantaneous(
    Spawn Count=&RandomRangeInt(Minimum=4, Maximum=5)
    Spawn Group=1
    Spawn Time=0.07
    Use Spawn Probability=false
    UseLoopCountLimit=false
   ) {module=SpawnBurst_Instantaneous001}
   &SpawnBurst_Instantaneous(
    Spawn Count=&RandomRangeInt(Minimum=2, Maximum=3)
    Spawn Group=2
    Spawn Time=0.05
    Use Spawn Probability=false
    UseLoopCountLimit=false
   ) {module=SpawnBurst_Instantaneous002}
  ]
  particleSpawnScript [
   &InitializeParticle(
    Color=&MergeHelper_ColorSwitch(
     0=<rgba 1 1 1 1>
     1=<rgba 1 1 1 1>
     2=&SelectLinearColorFromArray(
      Array Sampling Mode=::NewEnumerator0
      Direct Array Index=?
      Color Selection Array=DataInterface() {propertyValues {ColorData [<rgba 0.3304 0.2302 0.0488 1>, <rgba 0.4 0.2787 0.0591 1>]}}
      Random Seed=0
     )
     3=?
     Index=$Emitter.SpawnGroup
     Num=::NewEnumerator1
    )
    Lifetime=&RandomRangeFloat(Minimum=0.8, Maximum=2)
    Mesh Renderer Array Visibility Mode=::NewEnumerator3
    Mesh Renderer Info=DataInterface() {propertyValues {MeshRenderer=&self:Left_Drawers_Pages_0.NiagaraMeshRendererProperties_2}}
    Mesh Scale=&Multiply_VectorByFloat(
     Vector=<1 3 1>
     Float=&Multiply_Float(
      A=&Multiply_Float(
       A=&RandomRangeFloat(Minimum=1, Maximum=1.3)
       B=&MergeHelper_FloatSwitch(0=1, 1=1, 2=0.6, 3=?, Index=$Emitter.SpawnGroup, Num=::NewEnumerator1)
      )
      B=&FN_FrontEndFloatModifier(Gameplay=1, Front End=0.75)
     )
    )
    Mesh Scale Mode=::NewEnumerator4
    UsePositionOffset=false
    Write Material Random=true
   )
   &SocketLocation(
    SamplePosOffset=true
    Sampled Position Offset=<0 0 3>
    Socket Index=&SelectIntFromArray(
     Direct Array Index=&MakeCustomIntFromBool(
      Bool=&SetBoolByIntComparison(A=$Emitter.SpawnGroup, B=0, Comparison Type=::NewEnumerator4)
      True INT=&SelectIntFromArray(
       Array Sampling Mode=::NewEnumerator1
       Direct Array Index=?
       Int Selection Array=DataInterface() {propertyValues {IntData [0,1]}}
       Random Seed=0
      )
      False INT=&MakeCustomIntFromBool(
       Bool=&SetBoolByIntComparison(A=$Emitter.SpawnGroup, B=1, Comparison Type=::NewEnumerator4)
       True INT=&SelectIntFromArray(
        Array Sampling Mode=::NewEnumerator1
        Direct Array Index=?
        Int Selection Array=DataInterface() {propertyValues {IntData [2,3]}}
        Random Seed=0
       )
       False INT=4
      )
     )
     Int Selection Array=DataInterface() {propertyValues {IntData [1,2,3,4,5]}}
    )
    Socket Reader=$System.Socket Reader
    Socket Sampling Mode=::NewEnumerator3
    VelocitySpeedLimit=false
   )
   &InitialMeshOrientation(Mesh Orientation Mode=::NewEnumerator2, Rotation=<0.9 -0.52 0>, Rotation Coordinate Space=::NewEnumerator2, Use Rotation Vector=true)
   &AddVelocity(
    Additional Quaternion Rotation=false
    Cone Angle=&RandomRangeFloat(Minimum=0.1, Maximum=0.25)
    Cone Angle Mode:ENiagara_AngleInput=::NewEnumerator1
    Cone Axis=&MergeHelper_VectorSwitch(0=<0.25 0 1>, 1=<0.35 0 1>, 2=<0 0 1>, 3=?, Index=$Emitter.SpawnGroup, Num=::NewEnumerator1)
    Inner Cone Angle=0.5
    Module.UseConeFalloff=false
    Module.UseVelDistribution=false
    Override Seed=false
    Rotation Angle Type:ENiagara_AngleInput=::NewEnumerator1
    Rotation Mode=::NewEnumerator3
    Rotation Quaternion:Quat4f=$Output.SocketLocation.SampledRotation
    Velocity Mode=::NewEnumerator2
    Velocity Speed=&Multiply_Float(A=&RandomRangeFloat(Minimum=15, Maximum=125), B=4)
   )
   &ApplyInitialForces()
   &DynamicMaterialParameters(
    Module.Index0Float0Write=true
    Module.Index0Float1Write=true
    Module.Index0Float2Write=true
    Module.Index0Float3Write=true
    Module.Index1Float0Write=true
    Module.Index1Float1Write=true
    Module.Index1Float2Write=true
    Module.Index1Float3Write=true
    Write Parameter Index 1=::NewEnumerator1
   )
  ]
  particleUpdateScript [
   &ParticleState()
   &UpdateMeshOrientation(Rotation Rate=&RandomRangeFloat(Minimum=-1.5, Maximum=1.5), Rotation Vector=<1 1 1>)
   &LinearForce(Force=&MakeVector(
     X=0
     Y=&FloatFromCurve(
      FloatCurve=DataInterface() {propertyValues {Curve {keys [
          {arriveTangent=-1.2148, interpMode=RCIM_Cubic, leaveTangent=-1.2148, tangentMode=RCTM_User, time=0, value=0}
          {arriveTangent=-3.7513, interpMode=RCIM_Cubic, leaveTangent=-3.7513, tangentMode=RCTM_User, time=0.2, value=1}
          <at 0.7 0>
         ]}}}
      Scale Curve=&Multiply_Float(
       A=&FN_FrontEndFloatModifier(Gameplay=800, Front End=800)
       B=&Lerp_Float(A=-1, B=1, Alpha=$User.ForwardDot)
      )
     )
     Z=&FloatFromCurve(
      FloatCurve=DataInterface() {propertyValues {Curve {
         defaultValue=1
         keys [
          {arriveTangent=9.0725, interpMode=RCIM_Cubic, leaveTangent=9.0725, tangentMode=RCTM_User, time=0, value=0}
          {arriveTangent=-2.5249, interpMode=RCIM_Cubic, leaveTangent=-2.5249, tangentMode=RCTM_User, time=0.05, value=1}
          <at 0.1 0.5 RCIM_Cubic>
         ]
        }}}
      Scale Curve=&FN_FrontEndFloatModifier(Gameplay=-200, Front End=-200)
     )
    ))
   &ScaleMeshSize(Scale Factor=&FN_FrontEndVector3Modifier(
     Gameplay=&VectorFromFloat(Value=&FloatFromCurve(FloatCurve=DataInterface() {propertyValues {Curve {keys [
           {interpMode=RCIM_Cubic, tangentMode=RCTM_User, time=0.6, value=1}
           {arriveTangent=-3.2032, interpMode=RCIM_Cubic, leaveTangent=-3.2032, tangentMode=RCTM_User, time=1, value=0}
          ]}}}))
     Front End=&VectorFromFloat(Value=&FloatFromCurve(FloatCurve=DataInterface() {propertyValues {Curve {keys [
           <at 0.15 0 RCIM_Cubic>
           {arriveTangent=6.6667, interpMode=RCIM_Cubic, leaveTangent=6.6667, time=0.21, value=0.58}
           {interpMode=RCIM_Cubic, tangentMode=RCTM_User, time=0.3, value=1}
           {interpMode=RCIM_Cubic, tangentMode=RCTM_User, time=0.7, value=1}
           {interpMode=RCIM_Cubic, tangentMode=RCTM_User, time=1, value=0}
          ]}}}))
    ))
   &InheritSourceMovement(Applied Position Falloff Distance=25, ApplyForce=false, ApplyPosition=true, ApplyPositionFalloff=false)
   &Drag(
    Drag=&FloatFromCurve(FloatCurve=DataInterface() {propertyValues {Curve {keys [
         {interpMode=RCIM_Cubic, tangentMode=RCTM_Break, time=0, value=0}
         {interpMode=RCIM_Cubic, leaveTangent=-0.5451, tangentMode=RCTM_Break, time=0.2, value=1.013}
         {arriveTangent=-3.7088, interpMode=RCIM_Cubic, tangentMode=RCTM_Break, time=1, value=0}
        ]}}})
    Rotational Drag=0
    Use Linear Drag=true
    Use Rotational Drag=true
   )
   &CurlNoiseForce(Bias Noise=false, Noise Strength=&RandomRangeFloat(Minimum=100, Maximum=200), Pan Noise=false)
   &GravityForce(Gravity=&FN_FrontEndVector3Modifier(Gameplay=<0 0 -500>, Front End=<0 0 -500>))
   &AccelerationForce(
    Acceleration=&MakeVector(
     X=0
     Y=&Multiply_Float(A=300, B=&Lerp_Float(A=-1, B=1, Alpha=$User.ForwardDot))
     Z=-50
    )
    Coordinate Space=::Local
   )
   &SolveForcesAndVelocity(Clamp Velocity=false, Limit Acceleration=false)
   &ScaleColor(Scale RGB=&MakeVectorFromLinearColorRGB(Color=$System.TOD Color), ScaleA=true, ScaleRGB=true)
  ]
  rendererClasses [&NiagaraMeshRendererProperties]
  renderers [{rendererClass=&NiagaraMeshRendererProperties, rendererIndex=0}]
 }
 Right_Drawers_Pages {
  emitterUpdateScript [
   &EmitterState(
    Inactive Response=::NewEnumerator1
    Life Cycle Mode=::NewEnumerator1
    Loop Behavior=::NewEnumerator1
    Loop Delay=$System.LoopDelay
    Loop Duration=2
    UseLoopDelay=true
   )
   &SpawnBurst_Instantaneous(
    Spawn Count=&RandomRangeInt(Minimum=3, Maximum=4)
    Spawn Time=0.045
    Use Spawn Probability=false
    UseLoopCountLimit=false
   )
   &SpawnBurst_Instantaneous(
    Spawn Count=&RandomRangeInt(Minimum=4, Maximum=5)
    Spawn Group=1
    Spawn Time=0.04
    Use Spawn Probability=false
    UseLoopCountLimit=false
   ) {module=SpawnBurst_Instantaneous001}
   &SpawnBurst_Instantaneous(
    Spawn Count=&RandomRangeInt(Minimum=2, Maximum=3)
    Spawn Group=2
    Use Spawn Probability=false
    UseLoopCountLimit=false
   ) {module=SpawnBurst_Instantaneous002}
  ]
  particleSpawnScript [
   &InitializeParticle(
    Color=&MergeHelper_ColorSwitch(
     0=<rgba 1 1 1 1>
     1=<rgba 1 1 1 1>
     2=&SelectLinearColorFromArray(
      Array Sampling Mode=::NewEnumerator0
      Direct Array Index=?
      Color Selection Array=DataInterface() {propertyValues {ColorData [<rgba 0.3304 0.2302 0.0488 1>, <rgba 0.4 0.2787 0.0591 1>]}}
      Random Seed=0
     )
     3=?
     Index=$Emitter.SpawnGroup
     Num=::NewEnumerator1
    )
    Lifetime=&RandomRangeFloat(Minimum=0.8, Maximum=2)
    Mesh Renderer Array Visibility Mode=::NewEnumerator3
    Mesh Renderer Info=DataInterface() {propertyValues {MeshRenderer=&self:Right_Drawers_Pages_0.NiagaraMeshRendererProperties_2}}
    Mesh Scale=&Multiply_VectorByFloat(
     Vector=<1 3 1>
     Float=&Multiply_Float(
      A=&Multiply_Float(
       A=&RandomRangeFloat(Minimum=1, Maximum=1.3)
       B=&MergeHelper_FloatSwitch(0=1, 1=1, 2=0.6, 3=?, Index=$Emitter.SpawnGroup, Num=::NewEnumerator1)
      )
      B=&FN_FrontEndFloatModifier(Gameplay=1, Front End=0.75)
     )
    )
    Mesh Scale Mode=::NewEnumerator4
    UsePositionOffset=false
    Write Material Random=true
   )
   &SocketLocation(
    SamplePosOffset=true
    Sampled Position Offset=<0 0 3>
    Socket Index=&SelectIntFromArray(
     Direct Array Index=&MakeCustomIntFromBool(
      Bool=&SetBoolByIntComparison(A=$Emitter.SpawnGroup, B=0, Comparison Type=::NewEnumerator4)
      True INT=&SelectIntFromArray(
       Array Sampling Mode=::NewEnumerator1
       Direct Array Index=?
       Int Selection Array=DataInterface() {propertyValues {IntData [0,1]}}
       Random Seed=0
      )
      False INT=&MakeCustomIntFromBool(
       Bool=&SetBoolByIntComparison(A=$Emitter.SpawnGroup, B=1, Comparison Type=::NewEnumerator4)
       True INT=&SelectIntFromArray(
        Array Sampling Mode=::NewEnumerator1
        Direct Array Index=?
        Int Selection Array=DataInterface() {propertyValues {IntData [2,3]}}
        Random Seed=0
       )
       False INT=4
      )
     )
     Int Selection Array=DataInterface() {propertyValues {IntData [6,7,8,9,10]}}
    )
    Socket Reader=$System.Socket Reader
    Socket Sampling Mode=::NewEnumerator3
    VelocitySpeedLimit=false
   )
   &InitialMeshOrientation(Mesh Orientation Mode=::NewEnumerator2, Rotation=<0.9 -0.52 0>, Rotation Coordinate Space=::NewEnumerator2, Use Rotation Vector=true)
   &AddVelocity(
    Additional Quaternion Rotation=false
    Cone Angle=&RandomRangeFloat(Minimum=0.1, Maximum=0.25)
    Cone Angle Mode:ENiagara_AngleInput=::NewEnumerator1
    Cone Axis=&MergeHelper_VectorSwitch(0=<0.05 0 1>, 1=<0 0 1>, 2=<0 0 1>, 3=?, Index=$Emitter.SpawnGroup, Num=::NewEnumerator1)
    Inner Cone Angle=0.5
    Module.UseConeFalloff=false
    Module.UseVelDistribution=false
    Override Seed=false
    Rotation Angle Type:ENiagara_AngleInput=::NewEnumerator1
    Rotation Mode=::NewEnumerator3
    Rotation Quaternion:Quat4f=$Output.SocketLocation.SampledRotation
    Velocity Mode=::NewEnumerator2
    Velocity Speed=&Multiply_Float(A=&RandomRangeFloat(Minimum=15, Maximum=125), B=4)
   ) {module=AddVelocity001}
   &ApplyInitialForces()
   &DynamicMaterialParameters(
    Module.Index0Float0Write=true
    Module.Index0Float1Write=true
    Module.Index0Float2Write=true
    Module.Index0Float3Write=true
    Module.Index1Float0Write=true
    Module.Index1Float1Write=true
    Module.Index1Float2Write=true
    Module.Index1Float3Write=true
    Write Parameter Index 1=::NewEnumerator1
   )
  ]
  particleUpdateScript [
   &ParticleState()
   &UpdateMeshOrientation(Rotation Rate=&RandomRangeFloat(Minimum=-1.5, Maximum=1.5), Rotation Vector=<1 1 1>)
   &LinearForce(Force=&MakeVector(
     X=0
     Y=&FloatFromCurve(
      FloatCurve=DataInterface() {propertyValues {Curve {keys [
          {arriveTangent=-1.2148, interpMode=RCIM_Cubic, leaveTangent=-1.2148, tangentMode=RCTM_User, time=0, value=0}
          {arriveTangent=-3.7513, interpMode=RCIM_Cubic, leaveTangent=-3.7513, tangentMode=RCTM_User, time=0.2, value=1}
          <at 0.7 0>
         ]}}}
      Scale Curve=&Multiply_Float(A=800, B=&Lerp_Float(A=-1, B=1, Alpha=$User.ForwardDot))
     )
     Z=&FloatFromCurve(
      FloatCurve=DataInterface() {propertyValues {Curve {
         defaultValue=1
         keys [
          {arriveTangent=9.0725, interpMode=RCIM_Cubic, leaveTangent=9.0725, tangentMode=RCTM_User, time=0, value=0}
          {arriveTangent=-2.5249, interpMode=RCIM_Cubic, leaveTangent=-2.5249, tangentMode=RCTM_User, time=0.05, value=1}
          <at 0.1 0.5 RCIM_Cubic>
         ]
        }}}
      Scale Curve=-200
     )
    ))
   &ScaleMeshSize(Scale Factor=&FN_FrontEndVector3Modifier(
     Gameplay=&VectorFromFloat(Value=&FloatFromCurve(FloatCurve=DataInterface() {propertyValues {Curve {keys [
           {interpMode=RCIM_Cubic, tangentMode=RCTM_User, time=0.6, value=1}
           {arriveTangent=-3.2032, interpMode=RCIM_Cubic, leaveTangent=-3.2032, tangentMode=RCTM_User, time=1, value=0}
          ]}}}))
     Front End=&VectorFromFloat(Value=&FloatFromCurve(FloatCurve=DataInterface() {propertyValues {Curve {keys [
           <at 0.15 0 RCIM_Cubic>
           {arriveTangent=6.6667, interpMode=RCIM_Cubic, leaveTangent=6.6667, time=0.21, value=0.58}
           {interpMode=RCIM_Cubic, tangentMode=RCTM_User, time=0.3, value=1}
           {interpMode=RCIM_Cubic, tangentMode=RCTM_User, time=0.7, value=1}
           {interpMode=RCIM_Cubic, tangentMode=RCTM_User, time=1, value=0}
          ]}}}))
    ))
   &InheritSourceMovement(Applied Position Falloff Distance=25, ApplyForce=false, ApplyPosition=true, ApplyPositionFalloff=false)
   &Drag(
    Drag=&FloatFromCurve(FloatCurve=DataInterface() {propertyValues {Curve {keys [
         {interpMode=RCIM_Cubic, tangentMode=RCTM_Break, time=0, value=0}
         {interpMode=RCIM_Cubic, leaveTangent=-0.5451, tangentMode=RCTM_Break, time=0.2, value=1.013}
         {arriveTangent=-3.7088, interpMode=RCIM_Cubic, tangentMode=RCTM_Break, time=1, value=0}
        ]}}})
    Rotational Drag=0
    Use Linear Drag=true
    Use Rotational Drag=true
   )
   &CurlNoiseForce(Bias Noise=false, Noise Strength=&RandomRangeFloat(Minimum=100, Maximum=500), Pan Noise=false)
   &GravityForce(Gravity=<0 0 -600>)
   &AccelerationForce(
    Acceleration=&MakeVector(
     X=0
     Y=&Multiply_Float(A=300, B=&Lerp_Float(A=-1, B=1, Alpha=$User.ForwardDot))
     Z=-50
    )
    Coordinate Space=::Local
   )
   &SolveForcesAndVelocity(Clamp Velocity=false, Limit Acceleration=false)
   &ScaleColor(Scale RGB=&MakeVectorFromLinearColorRGB(Color=$System.TOD Color), ScaleA=true, ScaleRGB=true)
  ]
  rendererClasses [&NiagaraMeshRendererProperties]
  renderers [{rendererClass=&NiagaraMeshRendererProperties, rendererIndex=0}]
 }
}
summary {
 systemName=NS_Glider_UskThump_Deploy
 emitters [
  {emitterName=Right_Drawers_Pages, rendererClasses [&NiagaraMeshRendererProperties]}
  {emitterName=Left_Drawers_Pages, rendererClasses [&NiagaraMeshRendererProperties]}
 ]
 userVariables [
  User.bIsFrontEnd {type=NiagaraBool}
  User.bIsFrontEndPreview {type=NiagaraBool}
  User.ForwardDot {type=NiagaraFloat, defaultValue=0, description=""}
 ]
}
